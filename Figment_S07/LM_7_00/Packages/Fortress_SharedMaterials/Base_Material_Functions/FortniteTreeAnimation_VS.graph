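# Material function that produces an optional sine-based sway offset for trees, named as a vertex shader animation.
# Time plus a sine of object world XY drives a second sine, masked to R so only X affects phase, varying motion per object.
# The scaled wave moves X and Y equally with zero Z, weighted by one minus twice vertex-color red.
# A static switch input, off by default, outputs zero when disabled.
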MaterialFunction /Figment_S07/LM_7_00/Packages/Fortress_SharedMaterials/Base_Material_Functions/FortniteTreeAnimation_VS
graph {nodes [
  n1 = ObjectPositionWS()
  n2 = ComponentMask(None=@n1) {r=true}
  n3 = ComponentMask(None=@n1) {g=true}
  n4 = Add(A=Time(), B=Sine(None=AppendVector(A=@n2, B=@n3)))
  n5 = Multiply(A=ComponentMask(None=Sine(None=@n4) {period=2}) {r=true}, B=Constant() {r=3})
  n6 = AppendVector(A=AppendVector(A=@n5, B=@n5), B=Constant())
  n7 = VertexColor()
  n8 = OneMinus(None=Add(A=@n7.R, B=@n7.R))
  n9 = StaticBool()
  FunctionOutput(None=StaticSwitch(
    True=Multiply(A=@n6, B=@n8)
    False=Constant()
    Value=FunctionInput(Preview=@n9) {bUsePreviewValueAsDefault=true, inputName=Use Tree Vertex Shader, inputType=FunctionInput_StaticBool}
   ))
 ]}
