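# Material function that projects one parameter-driven decal in local space onto a material's base color, normals and specular masks.
# Decal position, axes and scale turn local position into planar UVs for its textures.
# Texture alpha, surface facing, texture group, UV edge masks and opacity form the blend; specular's lerp runs inverted, showing decal where the mask is zero.
# Static switches and feature-level switches pass inputs through when disabled or on ES2/ES3.1.
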MaterialFunction /Figment_S07/LM_7_00/Packages/Fortress_SharedMaterials/Base_Material_Functions/InternalDecal
graph {nodes [
  n1 = VectorParameter() {group=z_code, parameterName=Code_DecalLocalY, defaultValue=<rgba 0 0 0 1>}
  n2 = TextureObjectParameter() {group=z_code, parameterName=Code_DecalTexture_Normals, samplerType=SAMPLERTYPE_Normal, texture=/Engine/EngineMaterials/FlatNormal}
  n3 = Subtract(
   A=TransformPosition(None=WorldPosition() {worldPositionShaderOffset=WPT_ExcludeAllShaderOffsets}.XYZ) {transformSourceType=TRANSFORMPOSSOURCE_World}
   B=VectorParameter() {group=z_code, parameterName=Code_DecalLocalPosition, defaultValue=<rgba 0 0 0 1>}.RGB
  ) {desc=decal position centric}
  n4 = VectorParameter() {group=z_code, parameterName=Code_DecalLocalX, defaultValue=<rgba 0 0 0 1>}
  n5 = Multiply(A=DotProduct(A=@n3, B=@n4.RGB), B=Constant3Vector() {constant=<rgba 1 0 0 0>})
  n6 = Multiply(A=DotProduct(A=@n1.RGB, B=@n3), B=Constant3Vector() {constant=<rgba 0 1 0 0>})
  n7 = VectorParameter() {group=z_code, parameterName=Code_DecalLocalZ, defaultValue=<rgba 0 0 0 1>}
  n8 = DotProduct(A=@n3, B=@n7.RGB)
  n9 = Constant3Vector() {constant=<rgba 0 0 1 0>}
  n10 = Divide(
   A=Add(A=Add(A=@n5, B=@n6), B=Multiply(A=@n8, B=@n9))
   B=VectorParameter() {group=z_code, parameterName=Code_DecalScale, defaultValue=<rgba 1 1 1 1>}.RGB
  )
  n11 = OneMinus(None=ConstantBiasScale(None=ComponentMask(None=@n10) {b=true, g=true}))
  n12 = TextureSample(UVs=@n11, Tex=FunctionInput(Preview=@n2) {bUsePreviewValueAsDefault=true, inputName=Decal Normal, inputType=FunctionInput_Texture2D}) {samplerType=SAMPLERTYPE_Normal}
  n13 = Multiply(A=Multiply(A=@n1.RGB) {constB=-1}, B=@n12.R)
  n14 = Multiply(A=Multiply(A=@n7.RGB) {constB=-1}, B=@n12.G)
  n15 = Transform(None=Add(A=Add(A=@n13, B=@n14), B=Multiply(A=@n4.RGB, B=@n12.B))) {transformSourceType=TRANSFORMSOURCE_Local, transformType=TRANSFORM_Tangent}
  n16 = FunctionInput() {bUsePreviewValueAsDefault=true, inputName=Tangent Space Normal, previewValue=<0 0 1 1>}
  n17 = MaterialFunctionCall(BaseNormal=@n16, AdditionalNormal=@n15) {materialFunction=&BlendAngleCorrectedNormals}
  n18 = ScalarParameter() {defaultValue=0.5, desc="0= Full material replace 1= blend angle corrected normals", group=z_code, parameterName=Code_Decal_NormalBlendMethod}
  n19 = TextureSampleParameter2D(UVs=@n11) {group=z_code, parameterName=Code_DecalTexture_BaseColor, samplerSource=SSM_Clamp_WorldGroupSettings}
  n20 = Clamp(None=DotProduct(A=Transform(None=@n16) {transformType=TRANSFORM_Local}, B=@n4.RGB))
  n21 = ScalarParameter() {defaultValue=0.5, group=z_code, parameterName=Code_Decal_MaskingCheapContrast}
  n22 = VectorParameter() {group=z_code, parameterName=CODE_TextureGroupsToApplyCustomDecalsTo, defaultValue=<rgba 1 1 1 1>}
  n23 = AppendVector(A=@n22.RGB, B=@n22.RGB)
  n24 = FunctionInput() {inputName=Texture Group F4, inputType=FunctionInput_Vector4}
  n25 = MaterialFunctionCall(Float2=@n11) {materialFunction=&BreakOutFloat2Components}
  n26 = Constant() {r=0.5}
  n27 = Multiply(
   A=Multiply(
    A=DotProduct(A=@n23, B=@n24)
    B=MaterialFunctionCall(In=@n20, Contrast=@n21) {materialFunction=&CheapContrast}.Result
   )
   B=Multiply(A=SphereMask(A=@n25.R, B=@n26) {attenuationRadius=0.5}, B=SphereMask(A=@n25.G, B=@n26) {attenuationRadius=0.5})
  )
  n28 = ScalarParameter() {group=z_code, parameterName=Code_DecalParametersSet}
  n29 = Multiply(
   A=Multiply(A=@n19.A, B=Multiply(A=@n27, B=@n28)) {desc=Alpha Mask}
   B=ScalarParameter() {defaultValue=1, group=z_code, parameterName=Code_DecalOpacity}
  )
  n30 = StaticSwitchParameter(True=LinearInterpolate(A=@n16, B=LinearInterpolate(A=@n17.Result, B=@n15, Alpha=@n18), Alpha=@n29), False=@n16) {group=Decal, parameterName=Support Local Space Decal Normals}
  FunctionOutput(None=StaticSwitchParameter(True=FeatureLevelSwitch(Default=@n30, ES2=@n16, ES3_1=@n16, SM4_REMOVED=@n30), False=@n16) {group=Decal, parameterName=Use Local Space Decal}) {outputName=Normals}
  n31 = TextureSampleParameter2D(UVs=@n11) {
   group=z_code
   parameterName=Code_DecalTexture_Spec
   samplerSource=SSM_Clamp_WorldGroupSettings
   samplerType=SAMPLERTYPE_Masks
   texture=/Figment_S07/LM_7_00/Environments/Props/Transylvania/Trees/Textures/Temp_specular_Mask
  }
  n32 = FunctionInput() {bUsePreviewValueAsDefault=true, inputName=Base Material Specular Masks, previewValue=<0.5 0 0.5 1>}
  n33 = LinearInterpolate(A=@n31.RGB, B=@n32, Alpha=@n29)
  FunctionOutput(None=StaticSwitchParameter(True=FeatureLevelSwitch(Default=@n33, ES2=@n32, ES3_1=@n32, SM4_REMOVED=@n33), False=@n32) {group=Decal, parameterName=Use Local Space Decal}) {outputName=Specular Masks}
  n34 = FunctionInput() {inputName=Diffuse, inputType=FunctionInput_Vector4}
  n35 = AppendVector(A=LinearInterpolate(A=ComponentMask(None=@n34) {b=true, g=true, r=true}, B=@n19.RGB, Alpha=@n29), B=ComponentMask(None=@n34) {a=true})
  FunctionOutput(None=StaticSwitchParameter(True=FeatureLevelSwitch(Default=@n35, ES2=@n34, ES3_1=@n34, SM4_REMOVED=@n35), False=@n34) {group=Decal, parameterName=Use Local Space Decal}) {outputName="Base Color (Float 4)"}
  Multiply(A=@n20, B=@n20)
 ]}
