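# Material function that outputs one of up to ten numbered inputs, chosen by the scalar transitionFactor.
# A chain of If nodes compares the factor to whole-number thresholds, giving a hard step between inputs rather than a blend, despite the Lerp name.
# Nested static switches ("Use 2 Textures" through "Use 10 Textures") set the chain length at compile time, with the highest enabled one winning and input 1 as fallback.
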MaterialFunction /Figment_S07/LM_7_00/Packages/Fortress_SharedMaterials/Base_Material_Functions/LerpTexturesBasedOnInt
graph {nodes [
  n1 = Constant() {r=9}
  n2 = FunctionInput() {inputName=transitionFactor, inputType=FunctionInput_Scalar, sortPriority=-1}
  n3 = FunctionInput() {inputName=1}
  n4 = If(A=Constant() {r=1}, B=@n2, "A > B"=@n3, "A < B"=FunctionInput() {inputName=2, sortPriority=1}) {desc=0}
  n5 = If(A=Constant() {r=2}, B=@n2, "A > B"=@n4, "A < B"=FunctionInput() {inputName=3, sortPriority=2})
  n6 = If(A=Constant() {r=3}, B=@n2, "A > B"=@n5, "A < B"=FunctionInput() {inputName=4, sortPriority=3})
  n7 = If(A=Constant() {r=4}, B=@n2, "A > B"=@n6, "A < B"=FunctionInput() {inputName=5, sortPriority=4})
  n8 = If(A=Constant() {r=5}, B=@n2, "A > B"=@n7, "A < B"=FunctionInput() {inputName=6, sortPriority=5})
  n9 = If(A=Constant() {r=6}, B=@n2, "A > B"=@n8, "A < B"=FunctionInput() {inputName=7, sortPriority=6})
  n10 = If(A=Constant() {r=7}, B=@n2, "A > B"=@n9, "A < B"=FunctionInput() {inputName=8, sortPriority=7})
  n11 = If(A=Constant() {r=8}, B=@n2, "A > B"=@n10, "A < B"=FunctionInput() {inputName=9, sortPriority=8})
  n12 = FunctionInput() {inputName=10, sortPriority=9}
  n13 = StaticBool()
  n14 = StaticSwitch(
   True=@n4
   False=@n3
   Value=FunctionInput(Preview=@n13) {bUsePreviewValueAsDefault=true, inputName=Use 2 Textures, inputType=FunctionInput_StaticBool, sortPriority=10}
  )
  n15 = FunctionInput(Preview=@n13) {bUsePreviewValueAsDefault=true, inputName=Use 3 Textures, inputType=FunctionInput_StaticBool, sortPriority=11}
  n16 = StaticSwitch(
   True=@n7
   False=StaticSwitch(
    True=@n6
    False=StaticSwitch(True=@n5, False=@n14, Value=@n15)
    Value=FunctionInput(Preview=@n13) {bUsePreviewValueAsDefault=true, inputName=Use 4 Textures, inputType=FunctionInput_StaticBool, sortPriority=12}
   )
   Value=FunctionInput(Preview=@n13) {bUsePreviewValueAsDefault=true, inputName=Use 5 Textures, inputType=FunctionInput_StaticBool, sortPriority=13}
  )
  n17 = FunctionInput(Preview=@n13) {bUsePreviewValueAsDefault=true, inputName=Use 6 Textures, inputType=FunctionInput_StaticBool, sortPriority=14}
  n18 = StaticSwitch(
   True=@n10
   False=StaticSwitch(
    True=@n9
    False=StaticSwitch(True=@n8, False=@n16, Value=@n17)
    Value=FunctionInput(Preview=@n13) {bUsePreviewValueAsDefault=true, inputName=Use 7 Textures, inputType=FunctionInput_StaticBool, sortPriority=15}
   )
   Value=FunctionInput(Preview=@n13) {bUsePreviewValueAsDefault=true, inputName=Use 8 Textures, inputType=FunctionInput_StaticBool, sortPriority=16}
  )
  n19 = FunctionInput(Preview=@n13) {bUsePreviewValueAsDefault=true, inputName=Use 9 Textures, inputType=FunctionInput_StaticBool, sortPriority=17}
  FunctionOutput(None=StaticSwitch(
    True=If(A=@n1, B=@n2, "A > B"=@n11, "A < B"=@n12)
    False=StaticSwitch(True=@n11, False=@n18, Value=@n19)
    Value=FunctionInput(Preview=@n13) {bUsePreviewValueAsDefault=true, inputName=Use 10 Textures, inputType=FunctionInput_StaticBool, sortPriority=18}
   )) {outputName=Final Output, sortPriority=-2}
 ]}
