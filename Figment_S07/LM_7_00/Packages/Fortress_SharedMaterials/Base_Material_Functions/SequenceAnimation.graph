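# Material function that produces an animated emissive color from a texture, optionally cropped by UV corners.
# A looping time value, automatic or manual and optionally offset by actor position, drives either a hard on/off flicker or a texture-based sequence.
# The sequence uses SphereMask on texture alpha as a 1D moving band, looping or ping-ponging via LinearSine.
# Static switches gate the animation, texture multiplier and diffuse modulation.
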MaterialFunction /Figment_S07/LM_7_00/Packages/Fortress_SharedMaterials/Base_Material_Functions/SequenceAnimation
graph {nodes [
  FunctionInput() {bUsePreviewValueAsDefault=true, inputName="------------------- ", inputType=FunctionInput_StaticBool, sortPriority=48}
  FunctionInput() {bUsePreviewValueAsDefault=true, inputName=---------------------, inputType=FunctionInput_StaticBool, sortPriority=70}
  FunctionInput() {bUsePreviewValueAsDefault=true, inputName=-------------------, inputType=FunctionInput_StaticBool, sortPriority=50}
  n1 = FunctionInput(Preview=TextureObject()) {bUsePreviewValueAsDefault=true, inputName=Texture, inputType=FunctionInput_Texture2D}
  n2 = FunctionInput() {bUsePreviewValueAsDefault=true, inputName=UVPositioning UpperLeft, inputType=FunctionInput_Vector2, previewValue=<0 0 1 1>}
  n3 = FunctionInput() {bUsePreviewValueAsDefault=true, inputName=UVPositioning LowerRight, inputType=FunctionInput_Vector2, previewValue=<0 0 1 1>}
  n4 = MaterialFunctionCall!(UVs, TextureIn=@n1, UpperLeftCorner=@n2, LowerRightCorner=@n3, ExportFloat4=StaticBool() {value=true}) {materialFunction=/Engine/Functions/Engine_MaterialFunctions01/Texturing/TextureCropping}
  n5 = StaticBool() {value=true}
  n6 = TextureSample(Tex=@n1)
  n7 = StaticSwitch(
   True=StaticSwitch(
    True=@n4.CroppedMasked
    False=@n4.CroppedMasked
    Value=FunctionInput(Preview=@n5) {bUsePreviewValueAsDefault=true, inputName=Mask Cropped Emissive, inputType=FunctionInput_StaticBool}
   )
   False=AppendVector(A=@n6.RGB, B=@n6.RGB)
   Value=FunctionInput(Preview=StaticBool()) {bUsePreviewValueAsDefault=true, inputName=Manipulate Texture UVs, inputType=FunctionInput_StaticBool}
  )
  n8 = FunctionInput(Preview=MaterialFunctionCall(f3=ActorPositionWS()) {materialFunction=&AddComponents}.f3) {
   bUsePreviewValueAsDefault=true
   description="Get the actors 3 position axies, and add them together. Exposing this to save duplocate work in shader."
   inputName=ActorPositionAdd3
  }
  n9 = FunctionInput() {
   bUsePreviewValueAsDefault=true
   inputName="World Position Animation Offset Multiplier (WS)"
   inputType=FunctionInput_Scalar
   sortPriority=51
   previewValue=<0.01 0 0 1>
  }
  n10 = Divide(A=@n8, B=@n9)
  n11 = FunctionInput() {bUsePreviewValueAsDefault=true, inputName=Sequence animation speed, inputType=FunctionInput_Scalar, sortPriority=73, previewValue=<1 0 0 1>}
  n12 = StaticSwitch(
   True=FunctionInput() {bUsePreviewValueAsDefault=true, inputName=ManualTime, inputType=FunctionInput_Scalar, sortPriority=72}
   False=MaterialFunctionCall(Speed=@n11) {materialFunction=&TimeWithSpeedVariable}.Time
   Value=FunctionInput(Preview=StaticBool()) {bUsePreviewValueAsDefault=true, inputName=ControlTimeManually, inputType=FunctionInput_StaticBool, sortPriority=71}
  )
  n13 = FunctionInput(Preview=StaticBool()) {bUsePreviewValueAsDefault=true, inputName=Use World Position to offset the animation, inputType=FunctionInput_StaticBool, sortPriority=53}
  n14 = StaticSwitch(True=Frac(None=Add(A=@n10, B=@n12)), False=Frac(None=@n12), Value=@n13)
  n15 = FunctionInput() {bUsePreviewValueAsDefault=true, inputName=Animated Radius, inputType=FunctionInput_Scalar, sortPriority=54, previewValue=<1 0 0 1>}
  n16 = Multiply(A=@n14, B=Add(A=Add(A=@n15, B=@n15)))
  n17 = StaticBool()
  n18 = FunctionInput(Preview=StaticBool() {value=true}) {
   bUsePreviewValueAsDefault=true
   description="default = (T)"
   inputName="Loop(T) or Ping  Pong(F) Animation"
   inputType=FunctionInput_StaticBool
   sortPriority=60
  }
  n19 = FunctionInput() {bUsePreviewValueAsDefault=true, inputName=Animated Hardness, inputType=FunctionInput_Scalar, sortPriority=55, previewValue=<1 0 0 1>}
  n20 = SphereMask(
   A=ComponentMask(None=@n7) {a=true}
   B=StaticSwitch(
    True=Subtract(A=@n16, B=@n15)
    False=MaterialFunctionCall(Value=@n14, Sine Phase=@n17) {materialFunction=&LinearSine}.Linear Sine
    Value=@n18
   )
   Radius=@n15
   Hardness=@n19
  ) {attenuationRadius=0.1, hardnessPercent=25}
  n21 = FunctionInput() {bUsePreviewValueAsDefault=true, inputName=Animated emissive black out time, inputType=FunctionInput_Scalar, previewValue=<0.5 0 0 1>}
  n22 = Clamp(None=Ceil(None=Subtract(A=@n14, B=@n21)))
  n23 = FunctionInput(Preview=StaticBool()) {
   bUsePreviewValueAsDefault=true
   inputName="Animate Using Texture Based Animation(T) or a Flicker(F)"
   inputType=FunctionInput_StaticBool
   sortPriority=49
   previewValue=<1 0 0 1>
  }
  n24 = ComponentMask(None=@n7) {b=true, g=true, r=true}
  n25 = FunctionInput(Preview=StaticBool()) {bUsePreviewValueAsDefault=true, inputName=Modulate Texture, inputType=FunctionInput_StaticBool, previewValue=<1 0 0 1>}
  n26 = StaticSwitch(
   True=Multiply(A=@n24, B=FunctionInput() {bUsePreviewValueAsDefault=true, inputName=Texture Multiplier, previewValue=<1 0 0 1>})
   False=@n24
   Value=@n25
  )
  n27 = FunctionInput(Preview=StaticBool()) {bUsePreviewValueAsDefault=true, inputName=Animate, inputType=FunctionInput_StaticBool, sortPriority=49, previewValue=<1 0 0 1>}
  n28 = StaticSwitch(True=Multiply(A=StaticSwitch(True=@n20, False=@n22, Value=@n23), B=@n26), False=@n26, Value=@n27)
  n29 = FunctionInput() {bUsePreviewValueAsDefault=true, inputName=Diffuse, sortPriority=69, previewValue=<1 1 1 1>}
  n30 = FunctionInput(Preview=StaticBool()) {
   bUsePreviewValueAsDefault=true
   description=False
   inputName=Modulate with diffuse
   inputType=FunctionInput_StaticBool
   sortPriority=68
   previewValue=<1 0 0 1>
  }
  FunctionOutput(None=StaticSwitch(True=Multiply(A=@n28, B=@n29), False=@n28, Value=@n30))
  MaterialFunctionCall() {materialFunction=&ComponentPivotLocation}
  Multiply(A=@n22, B=@n14)
 ]}
